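# Skinned-mesh material for the HumanBeing backpack screen, showing a glitching emissive image with a blendable death state.
# A time-looped curve atlas row offsets a glitch mask and scales the UV distortion it applies to the FX2 red channel on UV1.
# PlayDeathAnim blends to the Y_Pos-shifted green channel, and inverted Fresnel masks the emissive.
# Standard character Pre/Post FX functions wrap this, with switchable clear coat, thin film and metal LUT layers.
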Material /BRCosmetics/Accessories/FORT_Backpacks/Backpack_M_MED_HumanBeing/Materials/M_HumanBeing_Backpack_Screen
graph {
 properties {
  bUseMaterialAttributes=true
  bUsedWithMeshDeformer=true
  bUsedWithSkeletalMesh=true
  pixelDepthOffsetMode=PDOM_Legacy
  usedShadingModels=Default Lit
 }
 nodes [
  n1 = MaterialFunctionCall() {materialFunction=&BaseFN_CharacterFunction_PreFX}
  n2 = TextureSample(Tex=TextureObjectParameter() {group=Skin FX, parameterName=SkinFX_Mask})
  n3 = StaticSwitchParameter(
   True=MaterialFunctionCall!(Alpha=StaticComponentMaskParameter(None=@n2.RGBA) {defaultR=true, group=Skin FX, parameterName=Clear Coat Channel}, In=@n1.Result) {materialFunction=/Game/Weapons/WeaponSkins/MaterialFunction/Clearcoat}.Result
   False=@n1.Result
  ) {group=Skin FX, parameterName=Use Clear Coat}
  n4 = StaticSwitchParameter(
   True=MaterialFunctionCall!(Alpha=StaticComponentMaskParameter(None=@n2.RGBA) {defaultR=true, group=Skin FX, parameterName=ThinFilm_Channel}, Material Attributes=@n3) {materialFunction=/Game/Weapons/WeaponSkins/MaterialFunction/ThinFilmFunction}.Result
   False=@n3
  ) {group=Skin FX, parameterName=Use Thin Film}
  n5 = MaterialFunctionCall(Mask=@n2.RGBA, Materials=@n4) {materialFunction=&MF_MetalsLUT}
  n6 = Power(Base=OneMinus(None=Fresnel() {exponent=1}))
  n7 = Time()
  n8 = CurveAtlasRowParameter(CurveTime=Frac(None=Multiply(A=@n7) {constB=0.5})) {
   atlas=/BRCosmetics/Accessories/FORT_Backpacks/Backpack_M_MED_HumanBeing/Textures/Curves/CA_HumanBeing_Backpack_Glitch
   curve=/BRCosmetics/Accessories/FORT_Backpacks/Backpack_M_MED_HumanBeing/Textures/Curves/C_HumanBeing_Backpack_Glitch
   group=HumanBeing Backpack
   parameterName=Glitch
  }
  n9 = TextureCoordinate() {coordinateIndex=1}
  n10 = Constant2Vector() {g=1.07, r=1.07}
  n11 = NamedRerouteDeclaration(None=Add(A=Multiply(A=@n9, B=@n10), B=Constant2Vector() {g=-0.07, r=-0.03})) {name=UV, nodeColor=<rgba 0.2656 1 0 1>}
  n12 = TextureSample(UVs=Add(A=@n11, B=MaterialFunctionCall(X=@n8.R, Y=@n8.R) {materialFunction=&MakeFloat2}.Result)) {
   samplerType=SAMPLERTYPE_Masks
   texture=/BRCosmetics/Accessories/FORT_Backpacks/Backpack_M_MED_HumanBeing/Textures/T_MED_HumanBeing_Backpack_Glitch
  }
  n13 = Multiply(A=@n8.R, B=ScalarParameter() {defaultValue=0.1, parameterName=UVDistortionStrength})
  n14 = TextureObjectParameter() {
   parameterName=HumanBeing FX2
   texture=/BRCosmetics/Accessories/FORT_Backpacks/Backpack_M_MED_HumanBeing/Textures/T_M_MED_HumanBeing_Backpack_FX2
  }
  n15 = TextureSample(UVs=Add(A=@n11, B=Multiply(A=@n12.R, B=@n13) {constB=0.1}), Tex=@n14) {
   samplerSource=SSM_Clamp_WorldGroupSettings
   texture=&T_M_MED_HumanBeing_Backpack_FX2
  }
  n16 = Constant()
  n17 = ScalarParameter() {defaultValue=0.7, group=HumanBeing Backpack, parameterName=Y_Pos}
  n18 = TextureSample(
   UVs=Add(
    A=NamedRerouteUsage()
    B=MaterialFunctionCall(X=@n16, Y=@n17) {materialFunction=&MakeFloat2}.Result
   )
   Tex=@n14
  ) {
   samplerSource=SSM_Clamp_WorldGroupSettings
   texture=&T_M_MED_HumanBeing_Backpack_FX2
  }
  n19 = LinearInterpolate(
   A=@n15.R
   B=Multiply(A=@n18.G, B=ScalarParameter() {defaultValue=1, group=HumanBeing Backpack, parameterName=DeathVisibility, sliderMax=1})
   Alpha=ScalarParameter() {group=HumanBeing Backpack, parameterName=PlayDeathAnim, sliderMax=1}
  )
  n20 = MaterialFunctionCall(
   AO=@n1.AO
   Distance Blend=@n1.AO
   Fresnel=@n1.AO
   Gmap=@n1.AO
   Pre Skinned Local Position=@n1.AO
   Material=SetMaterialAttributes(
    MaterialAttributes=StaticSwitchParameter(True=@n5.Result, False=@n4) {group=Skin FX, parameterName=UseMetalLUT}
    Emissive Color=Multiply(A=@n6, B=@n19)
   ) {attributeSetTypes [#1]}
  ) {materialFunction=&BaseFN_CharacterFunction_PostFX}
  MaterialFunctionCall(In=@n20.Result) {materialFunction=&CINE_MF_CharPostFX}
 ]
}
