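# Physics asset for the HumanBeing backpack cosmetic, defining jointed bodies for its dangling cord and knot.
# A base cord chain starts locked to the pack and forks at its last link into five-link cord and three-link knot branches.
# Tight angular limits generally loosen toward the branch tips, where boxes add bulkier collision shapes.
# Capsules on the pack, pelvis, spine and upper arms outline the wearer, with freely rotating constraints between body bones.
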PhysicsAsset /BRCosmetics/Accessories/FORT_Backpacks/Backpack_M_MED_HumanBeing/Meshes/M_MED_HumanBeing_Pack_Physics
bodies {
 dyn_cord_01 [{
   shapeType=Capsule
   length=1.9383
   radius=0.685
   shapeName=dyn_cord_01_capsule
   center=<1.1662 0.0828 0.0401>
   rotation=<rot 0 88.3281 -84.8166>
  }]
 dyn_cord_02 [{
   shapeType=Capsule
   length=4.1502
   radius=0.5998
   shapeName=dyn_cord_02_capsule
   center=<1.9906 0 -0.0291>
   rotation=<rot 89.7277 179.9987 179.9983>
  }]
 dyn_cord_03 [{shapeType=Capsule, length=3.7295, radius=0.7705, shapeName=dyn_cord_03_capsule, center=<2.3806 0 -0.0131>, rotation=<rot 87.4108 0 0>}]
 dyn_cord_04 [{
   shapeType=Capsule
   length=4
   radius=1
   shapeName=dyn_cord_04_capsule
   center=<2.5407 -0.1427 -0.0038>
   rotation=<rot 87.3305 -178.7627 179.9999>
  }]
 dyn_cord_05 [{shapeType=Box, extentX=4.5, extentY=2, extentZ=4, shapeName=dyn_cord_05_box, center=<3.2921 0 0.0632>, rotation=<rot 1.6112 1.2195 -6.1597>}]
 dyn_cord_base [{shapeType=Capsule, length=0.505, radius=0.505, shapeName=dyn_cord_base_capsule, rotation=<rot 0 -90 0>}]
 dyn_cord_base_01 [{
   shapeType=Capsule
   length=1.1033
   radius=0.5895
   shapeName=dyn_cord_base_01_capsule
   center=<0.6796 -0.4655 -0.4041>
   rotation=<rot -1.4933 -138.8998 -64.2631>
  }]
 dyn_cord_base_02 [{
   shapeType=Capsule
   length=2.3512
   radius=0.5944
   shapeName=dyn_cord_base_02_capsule
   center=<2.5244 -0.112 0.0343>
   rotation=<rot 0 86.1216 -89.2828>
  }]
 dyn_cord_base_03 [{
   shapeType=Capsule
   length=3.3864
   radius=0.5373
   shapeName=dyn_cord_base_03_capsule
   center=<1.8633 0.0673 0.0728>
   rotation=<rot 0 92.1531 -85.995>
  }]
 dyn_cord_base_04 [{
   shapeType=Capsule
   length=0.3696
   radius=1.5103
   shapeName=dyn_cord_base_04_capsule
   center=<0.8603 0 -0.171>
   rotation=<rot -82.7321 -180 -179.9998>
  }]
 dyn_knot_01 [{shapeType=Capsule, length=3.3154, radius=1.928, shapeName=dyn_knot_01_capsule, center=<2.5836 0.7695 0.1131>, rotation=<rot -90 0 0>}]
 dyn_knot_02 [
  {shapeType=Capsule, length=2.1868, radius=0.9605, shapeName=dyn_knot_02_capsule, center=<1.4635 0 0>, rotation=<rot 90 0 0>}
  {
   shapeType=Box
   extentX=3.9251
   extentY=3.0197
   extentZ=4.7666
   shapeName=dyn_knot_02_box
   center=<2.0563 0.0773 0.6316>
   rotation=<rot 2.5889 2.818 -42.7728>
  }
 ]
 dyn_knot_03 [
  {
   shapeType=Capsule
   length=2.2301
   radius=0.8577
   shapeName=dyn_knot_03_capsule
   center=<0.9357 -0.1907 -0.106>
   rotation=<rot -37.4085 -89.9999 -93.0774>
  }
  {
   shapeType=Box
   extentX=3.6786
   extentY=2.9645
   extentZ=4.8674
   shapeName=dyn_knot_03_box
   center=<1.6909 0.3628 0.6424>
   rotation=<rot 1.5528 1.6068 -37.4552>
  }
 ]
 pack [
  {shapeType=Capsule, length=20.4186, radius=2, shapeName=pack_capsule, center=<-8.7681 -4.963 -1.5357>, rotation=<rot 74.9396 -0.0807 -0.0836>}
  {
   shapeType=Capsule
   length=16.8753
   radius=2
   shapeName=pack_capsule
   center=<8.5667 -4.9709 -1.5897>
   rotation=<rot -74.9396 -0.0807 -0.0836>
  }
  {
   shapeType=Capsule
   length=17.4989
   radius=2.0666
   shapeName=pack_capsule
   center=<6.981 -0.0974 -1.9414>
   rotation=<rot -74.9543 0.0006 32.7851>
  }
  {
   shapeType=Capsule
   length=18.65
   radius=2.0845
   shapeName=pack_capsule
   center=<-8.51 -4.9899 6.9199>
   rotation=<rot -73.9422 0.0178 -0.0186>
  }
  {
   shapeType=Capsule
   length=17.9093
   radius=2.2365
   shapeName=pack_capsule
   center=<8.5634 -4.9062 7.0769>
   rotation=<rot 74.6492 -0.0004 -0.0004>
  }
 ]
 pelvis [{shapeType=Capsule, length=13.7137, radius=10.2565, center=<0 1.1083 0>, rotation=<rot 0 -90 0>}]
 spine_01 [{shapeType=Capsule, length=10.8871, radius=10.1791, center=<0 -0.3001 -0.0001>}]
 spine_02 [{shapeType=Capsule, length=15.4804, radius=9.7089, center=<0 -2.2851 -0.0001>}]
 spine_03 [{shapeType=Capsule, length=13.1327, radius=9.3754, center=<0 -2.9669 -0.0001>}]
 spine_04 [{shapeType=Capsule, length=11.6229, radius=8.0931, center=<0.6532 -3.22 0>}]
 spine_05 [{shapeType=Capsule, length=18.1163, radius=8.7969, rotation=<rot 0 -90 0>}]
 upperarm_l [{shapeType=Capsule, length=18.0812, radius=4.1347, center=<10.8056 -0.4936 0.3456>, rotation=<rot 90 0 0>}]
 upperarm_r [{shapeType=Capsule, length=18.0812, radius=4.1347, center=<-10.8056 0.4936 -0.3456>, rotation=<rot 90 0 0>}]
}
constraints [
 {
  bone1Name=spine_01
  bone2Name=pelvis
  swing1LimitDegrees=45
  swing1Motion=Free
  swing2LimitDegrees=45
  swing2Motion=Free
  twistLimitDegrees=45
  twistMotion=Free
 }
 {
  bone1Name=spine_02
  bone2Name=spine_01
  swing1LimitDegrees=45
  swing1Motion=Free
  swing2LimitDegrees=45
  swing2Motion=Free
  twistLimitDegrees=45
  twistMotion=Free
 }
 {
  bone1Name=spine_03
  bone2Name=spine_02
  swing1LimitDegrees=45
  swing1Motion=Free
  swing2LimitDegrees=45
  swing2Motion=Free
  twistLimitDegrees=45
  twistMotion=Free
 }
 {
  bone1Name=spine_04
  bone2Name=spine_03
  swing1LimitDegrees=45
  swing1Motion=Free
  swing2LimitDegrees=45
  swing2Motion=Free
  twistLimitDegrees=45
  twistMotion=Free
 }
 {
  bone1Name=spine_05
  bone2Name=spine_04
  swing1LimitDegrees=45
  swing1Motion=Free
  swing2LimitDegrees=45
  swing2Motion=Free
  twistLimitDegrees=45
  twistMotion=Free
 }
 {
  bone1Name=upperarm_l
  bone2Name=spine_05
  swing1LimitDegrees=45
  swing1Motion=Free
  swing2LimitDegrees=45
  swing2Motion=Free
  twistLimitDegrees=45
  twistMotion=Free
 }
 {
  bone1Name=upperarm_r
  bone2Name=spine_05
  swing1LimitDegrees=45
  swing1Motion=Free
  swing2LimitDegrees=45
  swing2Motion=Free
  twistLimitDegrees=45
  twistMotion=Free
 }
 {bone1Name=dyn_cord_03, bone2Name=dyn_cord_02, swing1LimitDegrees=16, swing2LimitDegrees=16, twistLimitDegrees=4}
 {bone1Name=dyn_cord_05, bone2Name=dyn_cord_04, swing1LimitDegrees=19, swing2LimitDegrees=19, twistLimitDegrees=4}
 {bone1Name=dyn_cord_04, bone2Name=dyn_cord_03, swing1LimitDegrees=18, swing2LimitDegrees=18, twistLimitDegrees=4}
 {bone1Name=dyn_cord_02, bone2Name=dyn_cord_01, swing1LimitDegrees=14, swing2LimitDegrees=14, twistLimitDegrees=3}
 {
  bone1Name=dyn_cord_base
  bone2Name=pack
  swing1LimitDegrees=45
  swing1Motion=Locked
  swing2LimitDegrees=45
  swing2Motion=Locked
  twistLimitDegrees=45
  twistMotion=Locked
 }
 {
  bone1Name=dyn_cord_base_01
  bone2Name=dyn_cord_base
  swing1LimitDegrees=45
  swing1Motion=Locked
  swing2LimitDegrees=45
  swing2Motion=Locked
  twistLimitDegrees=15
 }
 {
  bone1Name=dyn_cord_base_02
  bone2Name=dyn_cord_base_01
  swing1LimitDegrees=2
  swing2LimitDegrees=13
  twistLimitDegrees=45
  twistMotion=Locked
 }
 {
  bone1Name=dyn_cord_base_03
  bone2Name=dyn_cord_base_02
  swing1LimitDegrees=3
  swing2LimitDegrees=13
  twistLimitDegrees=45
  twistMotion=Locked
 }
 {
  bone1Name=dyn_cord_base_04
  bone2Name=dyn_cord_base_03
  swing1LimitDegrees=2
  swing2LimitDegrees=14
  twistLimitDegrees=45
  twistMotion=Locked
 }
 {bone1Name=dyn_cord_01, bone2Name=dyn_cord_base_04, swing1LimitDegrees=1, swing2LimitDegrees=12, twistLimitDegrees=2}
 {bone1Name=dyn_knot_01, bone2Name=dyn_cord_base_04, swing1LimitDegrees=2, swing2LimitDegrees=10, twistLimitDegrees=2}
 {bone1Name=dyn_knot_02, bone2Name=dyn_knot_01, swing1LimitDegrees=10, swing2LimitDegrees=5, twistLimitDegrees=2}
 {bone1Name=dyn_knot_03, bone2Name=dyn_knot_02, swing1LimitDegrees=10, swing2LimitDegrees=7, twistLimitDegrees=3}
]
